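# Sound cue for the close-range audio of the Cotton Candy back bling cosmetic.
# A single wave player feeds a Delay node, left at its default settings, as the cue's root.
# It plays at reduced volume through the back bling sound class, using close-range pet attenuation and the reactive back bling audio property sheet.
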SoundCue /BRCosmetics/Sounds/Backblings/CottonCandy/BackBling_CottonCandy_Close_Cue
root=Delay_0
nodes [
 {id=Delay_0, children [WavePlayer_1]}
 {id=WavePlayer_1, properties {soundWaveAssetPtr=&/BackBling_CottonCandy_Close}}
]
properties {
 attenuationSettings=&Fort_Pets_Close_Athena_Att
 audioPropertiesSheet=/BRCosmetics/Sounds/Settings/PropertySheets/Reactive/APS_Cosmetic_Reactive_Backbling
 maxDistance=1800.0001
 soundClassObject=/Game/Sounds/FortSoundClassesAndModes/Fort_Backblings
 volumeMultiplier=0.2
}
